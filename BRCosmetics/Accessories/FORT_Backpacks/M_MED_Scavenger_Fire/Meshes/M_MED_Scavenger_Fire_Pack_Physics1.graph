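# Physics asset for the M_MED_Scavenger_Fire backpack cosmetic, covering the dynamic bones of its bat and chains.
# Each of the four bones has a single capsule body: a master bat bone, the bat, and high and low chains.
# Constraints join the bat to its master bone and both chains to the bat, with small swing and twist limit angles.
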PhysicsAsset /BRCosmetics/Accessories/FORT_Backpacks/M_MED_Scavenger_Fire/Meshes/M_MED_Scavenger_Fire_Pack_Physics1
bodies {
 dyn_bat [{shapeType=Capsule, length=15.4982, radius=4.1441, center=<11.2059 -1.9967 0.0002>, rotation=<rot 90 0 1>}]
 dyn_bat_MSTR [{shapeType=Capsule, length=4.25, radius=3.255, center=<0 0 -4.0753>, rotation=<rot 0 -90 0>}]
 dyn_chains_high [{shapeType=Capsule, length=4.2971, radius=1.9085, center=<-0.8693 -5.0786 1.9357>, rotation=<rot -19.6749 -89.9999 -27.6426>}]
 dyn_chains_low [{shapeType=Capsule, length=4.1652, radius=1.8147, center=<-1.9908 -2.5484 0.0002>, rotation=<rot 9.5308 -90 -3.9174>}]
}
constraints [
 {bone1Name=dyn_bat, bone2Name=dyn_bat_MSTR, swing1LimitDegrees=4, swing2LimitDegrees=18, twistLimitDegrees=20}
 {bone1Name=dyn_chains_low, bone2Name=dyn_bat, swing1LimitDegrees=8, swing2LimitDegrees=8, twistLimitDegrees=2}
 {bone1Name=dyn_chains_high, bone2Name=dyn_bat, swing1LimitDegrees=10, swing2LimitDegrees=10, twistLimitDegrees=2}
]
