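# Animation Blueprint for the M_MED_Scavenger_Fire backpack cosmetic that adds physics-driven secondary motion.
# The AnimGraph starts from the reference pose, additively nudges the pack bone, then runs a RigidBody simulation in root-bone space with overridden gravity and scaled component motion.
# The update event lowers simulation alpha in the front end.
# It also sets Gravity Override by skydiving state, but the AnimGraph never reads that variable.
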AnimBlueprint /BRCosmetics/Accessories/FORT_Backpacks/M_MED_Scavenger_Fire/Meshes/M_MED_Scavenger_Fire_Pack_AnimBP
target_skeleton=&/M_MED_Scavenger_Fire_Pack_Skeleton
graphs [
 AnimGraph {nodes [
   n1 = "Transform (Modify) Bone - Bone: pack"(ComponentPose=Local To Component(LocalPose=Local Space Ref Pose()), Translation="0,1,0", Rotation="0,0,0") {
    translationMode=BMM_Additive
    translationSpace=BCS_BoneSpace
    alphaBoolBlend {blendOption=Linear}
    boneToModify {boneName=pack}
    translation {y=1}
   }
   n2 = RigidBody!(ComponentPose=@n1, Alpha=Get FE Alpha() -> (FE Alpha:double), OverrideWorldGravity:Vector="0,0,-980") -> (Pose) {
    bOverrideWorldGravity=true
    overridePhysicsAsset=&/M_MED_Scavenger_Fire_Pack_Physics1
    simulationSpace=BaseBoneSpace
    alphaBoolBlend {blendOption=Linear}
    baseBoneRef {boneName=root}
    componentAppliedLinearAccClamp=<3000 3000 3000>
    componentLinearAccScale=<0.3 0.3 0.3>
    componentLinearVelScale=<0.15 0.15 0.15>
    overrideWorldGravity {z=-980}
   }
   Output Pose(Result=Component To Local(ComponentPose=@n2)) {layerGroup=None}
  ]}
 EventGraph {nodes [
   Event BlueprintUpdateAnimation()
   > n1 = Sequence()
   > n2 = Branch(Condition=Get bIsSkyDiving())
   > Set Gravity Override(Gravity Override="0.05,0.05,0.05")
   Set Gravity Override(execute=@n2.else, Gravity Override="0.3,0.3,0.3")
   n3 = Branch(execute=@n1.then_1, Condition=Get bIsInFrontEnd())
   > Set FE Alpha(FE Alpha:double=0.5) -> (Output_Get:double)
   Set FE Alpha(execute=@n3.else, FE Alpha=1)
  ]}
]
variables [
 FE Alpha {type=number}
 Gravity Override {type=Vector}
]
